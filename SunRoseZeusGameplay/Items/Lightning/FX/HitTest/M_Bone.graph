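# Translucent, depth-test-disabled bone material in the lightning item's HitTest FX folder, flagged for skeletal, cloth and Niagara meshes.
# BoneColor is shaded by absolute camera-to-normal-map dot product, alongside a colored Fresnel term.
# A SceneDepth versus PixelDepth comparison builds a mask of unoccluded pixels.
# An OffsetToViewPos parameter scales a view-relative offset vector; no material output connections appear.
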Material /SunRoseZeusGameplay/Items/Lightning/FX/HitTest/M_Bone
graph {
 properties {
  bDisableDepthTest=true
  bUsedWithClothing=true
  bUsedWithMeshDeformer=true
  bUsedWithNiagaraMeshParticles=true
  bUsedWithSkeletalMesh=true
  blendMode=BLEND_Translucent
  pixelDepthOffsetMode=PDOM_Legacy
  usedShadingModels=Default Lit
 }
 nodes [
  n1 = WorldPosition()
  n2 = TransformPosition(None=AppendVector(A=ScreenPosition().ViewportUV, B=Constant())) {transformSourceType=TRANSFORMPOSSOURCE_View, transformType=TRANSFORMPOSSOURCE_World}
  Multiply(A=Normalize(VectorInput=Subtract(A=@n1.XYZ, B=@n2)), B=ScalarParameter() {defaultValue=-20, parameterName=OffsetToViewPos})
  n3 = VectorParameter() {parameterName=Fresnel, defaultValue=<rgba 5 0.04 0 1>}
  Multiply(
   A=VectorParameter() {parameterName=fresnellColor, defaultValue=<rgba 0 1.226 5 1>}.RGB
   B=Fresnel(ExponentIn=@n3.R, BaseReflectFractionIn=@n3.R)
  )
  n4 = CameraVectorWS()
  n5 = TextureSampleParameter2D() {
   parameterName=Normal
   samplerType=SAMPLERTYPE_Normal
   texture=/Game/Characters/Player/Male/Medium/Bodies/M_MED_Keychain/Textures/T_KeyChain_Body_N
  }
  n6 = Transform(None=@n5.RGB)
  Multiply(A=VectorParameter() {parameterName=BoneColor, defaultValue=<rgba 0.5 0.5 0.5 1>}.RGB, B=Abs(None=DotProduct(A=@n4, B=@n6)))
  n7 = Constant()
  Multiply(A=Constant() {r=1}, B=If(A=SceneDepth().R, B=PixelDepth().R, "A > B"=Constant() {r=1}, "A == B"=@n7, "A < B"=@n7))
 ]
}
