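# Additive unlit material for Niagara sprites, named as a hit glow among the Zeus Lightning item's hit-test effects.
# A soft flare texture is sampled through the shared FXCommon_TextureSample function, with parameters for texture, UV scale and pan.
# The sample is fully desaturated and multiplied by particle alpha, and particle RGB is not used.
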Material /SunRoseZeusGameplay/Items/Lightning/FX/HitTest/M_HitGlow
graph {
 properties {
  bUsedWithNiagaraSprites=true
  blendMode=BLEND_Additive
  pixelDepthOffsetMode=PDOM_Legacy
  shadingModel=MSM_Unlit
  translucencyPass=MTP_AfterMotionBlur
  usedShadingModels=Unlit
 }
 nodes [
  n1 = TextureObjectParameter() {group=GlowT, parameterName=GlowT_Texture, sortPriority=0, texture=/Game/Effects/Fort_Effects/Textures/Flares/T_Flare_Soft}
  n2 = VectorParameter() {group=GlowT, parameterName=GlowT_UVControl, channelNames {a=VPan, b=UPan, g=VScale, r=UScale}, defaultValue=<rgba 1 1 0 0>}
  Multiply(
   A=ParticleColor().A
   B=Desaturation(
    None=MaterialFunctionCall(Texture=@n1, UVControl_RGB=@n2, UVControl_A=@n2) {materialFunction=&FXCommon_TextureSample}.RGB
    Fraction=Constant() {r=1}
   )
  )
 ]
}
