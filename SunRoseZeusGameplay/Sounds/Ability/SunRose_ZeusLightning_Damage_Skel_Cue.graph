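# Sound cue for the Zeus lightning damage ability sound, playing electrocution effects.
# A random node picks one of four equally weighted electrocute waves and feeds a modulator at the root for variation.
# It uses a shared lightning damage attenuation asset and a received-hit-notification sound class, with boosted volume.
# Overridden concurrency caps simultaneous plays per owner and sets a short retrigger time.
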SoundCue /SunRoseZeusGameplay/Sounds/Ability/SunRose_ZeusLightning_Damage_Skel_Cue
root=Modulator_0
nodes [
 {id=Modulator_0, children [Random_0]}
 {id=Random_0, children [WavePlayer_0,WavePlayer_1,WavePlayer_2,WavePlayer_3], properties {weights [1,1,1,1]}}
 {id=WavePlayer_0, properties {soundWaveAssetPtr=/SunRoseZeusGameplay/Sounds/Ability/Wavs/Sunrose_Zeus_Lightning_Electrocute_04}}
 {id=WavePlayer_1, properties {soundWaveAssetPtr=/SunRoseZeusGameplay/Sounds/Ability/Wavs/Sunrose_Zeus_Lightning_Electrocute_01}}
 {id=WavePlayer_2, properties {soundWaveAssetPtr=/SunRoseZeusGameplay/Sounds/Ability/Wavs/Sunrose_Zeus_Lightning_Electrocute_02}}
 {id=WavePlayer_3, properties {soundWaveAssetPtr=/SunRoseZeusGameplay/Sounds/Ability/Wavs/Sunrose_Zeus_Lightning_Electrocute_03}}
]
properties {
 attenuationSettings=&/SunRose_ZeusLightning_Damage_Att
 bOverrideConcurrency=true
 maxDistance=1100
 soundClassObject=/Game/Sounds/FortSoundClassesAndModes/Fort_2DHitNotifications_Receive_PSM
 volumeMultiplier=1.5
 attenuationOverrides {distanceAlgorithm=NaturalSound, falloffDistance=3000, attenuationShapeExtents {x=1000}}
 concurrencyOverrides {bLimitToOwner=true, maxCount=2, retriggerTime=0.2}
}
